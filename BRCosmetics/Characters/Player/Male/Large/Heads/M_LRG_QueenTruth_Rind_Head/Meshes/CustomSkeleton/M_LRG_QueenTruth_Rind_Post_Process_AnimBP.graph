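# Post-process Animation Blueprint for the large male QueenTruth Rind character head, targeting the Fortnite base head skeleton.
# The AnimGraph runs the incoming pose through a single Control Rig node using the head's dedicated rig, then outputs the result.
# The rig's isEmoting input reads bIsPlayingEmote, which is not declared in this Blueprint.
# The EventGraph is empty, so all the work happens in the rig.
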AnimBlueprint /BRCosmetics/Characters/Player/Male/Large/Heads/M_LRG_QueenTruth_Rind_Head/Meshes/CustomSkeleton/M_LRG_QueenTruth_Rind_Post_Process_AnimBP
target_skeleton=/Game/Characters/Player/Common/Fortnite_Base_Head/Fortnite_Base_Head_Export_Skeleton
graphs [
 AnimGraph {nodes [
   n1 = Control Rig!(Source=Input Pose(), Alpha, isEmoting:bool=Get bIsPlayingEmote()) -> (Pose) {controlRigAssetReference {blueprintRigClass=/BRCosmetics/Characters/Player/Male/Large/Heads/M_LRG_QueenTruth_Rind_Head/Meshes/M_LRG_QueenTruth_Rind_Head_CtrlRig.M_LRG_QueenTruth_Rind_Head_CtrlRig_C}}
   Output Pose(Result=@n1)
  ]}
 EventGraph {links [], nodes []}
]
variables []
